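# Physics asset for the CosmicTrellis Goth medium male player body, covering four dynamic jacket bones (left, right, back left, back right) and both thighs.
# Box bodies on the side panels and capsules on the back panels are each constrained to a small sphere on their jacket base bone.
# Constraint limits zero out swing1 and twist, leaving a narrow swing2 hinge for each panel.
# Unconstrained thigh capsules provide leg shapes, likely as colliders for the jacket.
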PhysicsAsset /BRCosmetics/Characters/Player/Male/Medium/Bodies/M_MED_CosmicTrellis_Goth/Meshes/M_MED_CosmicTrellis_Goth_Physics
bodies {
 dyn_jacket_01_back_l [{
   shapeType=Capsule
   length=7.4224
   radius=4.2127
   shapeName=dyn_jacket_01_back_l_capsule
   center=<5.2534 0.9031 -3.8597>
   rotation=<rot 90 0 0>
  }]
 dyn_jacket_01_back_r [{
   shapeType=Capsule
   length=7.3403
   radius=4.339
   shapeName=dyn_jacket_01_back_r_capsule
   center=<5.1969 -0.8011 -4.0287>
   rotation=<rot 90 0 0>
  }]
 dyn_jacket_01_l [{
   shapeType=Box
   extentX=7
   extentY=2.3757
   extentZ=12
   shapeName=dyn_jacket_01_l_box
   center=<8.3478 -0.7025 -1.0096>
   rotation=<rot 0.0001 -0.0005 95.3904>
  }]
 dyn_jacket_01_r [{
   shapeType=Box
   extentX=9.0589
   extentY=2.5724
   extentZ=13.2927
   shapeName=dyn_jacket_01_r_box
   center=<7.2509 1.1633 -1.2006>
   rotation=<rot 0 0 90>
  }]
 dyn_jacket_basa_back_l [{shapeType=Sphere, radius=0.5, shapeName=dyn_jacket_basa_back_l_sphere}]
 dyn_jacket_basa_back_r [{shapeType=Sphere, radius=0.5, shapeName=dyn_jacket_basa_back_r_sphere}]
 dyn_jacket_basa_l [{shapeType=Sphere, radius=0.5, shapeName=dyn_jacket_basa_l_sphere}]
 dyn_jacket_basa_r [{shapeType=Sphere, radius=0.5, shapeName=dyn_jacket_basa_r_sphere}]
 thigh_l [{shapeType=Capsule, length=30, radius=7, center=<-15.4683 0.3228 1.5363>, rotation=<rot 0.0001 -87.4725 -86.1678>}]
 thigh_r [{shapeType=Capsule, length=21.8066, radius=7, center=<19.0328 0.1639 -1.6202>, rotation=<rot 0 -87.4723 -88.6663>}]
}
constraints [
 {bone1Name=dyn_jacket_01_back_r, bone2Name=dyn_jacket_basa_back_r, swing1LimitDegrees=0, swing2LimitDegrees=25, twistLimitDegrees=0}
 {bone1Name=dyn_jacket_01_back_l, bone2Name=dyn_jacket_basa_back_l, swing1LimitDegrees=0, swing2LimitDegrees=25, twistLimitDegrees=0}
 {bone1Name=dyn_jacket_01_r, bone2Name=dyn_jacket_basa_r, swing1LimitDegrees=0, swing2LimitDegrees=20, twistLimitDegrees=0}
 {bone1Name=dyn_jacket_01_l, bone2Name=dyn_jacket_basa_l, swing1LimitDegrees=0, swing2LimitDegrees=15, twistLimitDegrees=0}
]
