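# Animation Blueprint for the CosmicTrellis Goth male medium cosmetic body, adding procedural and physics motion to the parent's pose.
# The copied pose feeds a Control Rig driven by crouch, emote, DBNO and melee flags, then a pelvis-space RigidBody simulation.
# Each update sets the melee flag from the weapon core animation and computes the simulation's gravity override.
# The gravity function uses the parameter asset from female PearlTaffyOpera.
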AnimBlueprint /BRCosmetics/Characters/Player/Male/Medium/Bodies/M_MED_CosmicTrellis_Goth/Meshes/M_MED_CosmicTrellis_Goth_AnimBP
target_skeleton=&/M_MED_CosmicTrellis_Goth_Skeleton
graphs [
 AnimGraph {nodes [
   n1 = Control Rig!(
    Source=Copy Pose From Mesh() {bUseAttachedParent=true}
    Alpha
    Crounch:bool=Get bIsCrouching()
    emote:bool=Get bIsPlayingEmote()
    DBNO:bool=Get bIsDBNO()
    melee:bool=Get Melee() -> (Melee:bool)
   ) -> (Pose) {controlRigAssetReference {controlRigAsset=&/M_MED_CosmicTrellis_Goth_CtrlRig}}
   n2 = RigidBody!(ComponentPose=Local To Component(LocalPose=@n1), Alpha, OverrideWorldGravity:Vector=Get Gravity Override(), ExternalForce) -> (Pose) {
    bOverrideWorldGravity=true
    overridePhysicsAsset=&/M_MED_CosmicTrellis_Goth_Physics
    simulationSpace=BaseBoneSpace
    alphaBoolBlend {blendOption=Linear}
    baseBoneRef {boneName=pelvis}
    simSpaceSettings {maxAngularAcceleration=500, maxAngularVelocity=3, maxLinearAcceleration=400, maxLinearVelocity=250, velocityScaleZ=0.7, worldAlpha=1}
   }
   Output Pose(Result=Component To Local(ComponentPose=@n2)) {layerGroup=None}
  ]}
 EventGraph {nodes [
   Event BlueprintUpdateAnimation()
   > n1 = "Gravity Override State Function (Parameter Asset) (Nativized)"(
    CharacterPartAnimInstance=Self-Reference()
    GameTimeInSeconds=GetGameTimeInSeconds()
    Parameters=/BRCosmetics/Characters/Player/Female/Medium/Bodies/F_MED_PearlTaffyOpera/Meshes/F_MED_PearlTaffyOpera_Parameters
   )
   > Set Gravity Override(Gravity Override=@n1)
   > Set Melee(Melee:bool="Equal (Enum)"(A:EFortWeaponCoreAnimation Enum=Get WeaponCoreAnim!() -> (WeaponCoreAnim), B:EFortWeaponCoreAnimation Enum=Melee)) -> (Output_Get:bool)
  ]}
]
variables [
 Gravity Override {type=Vector}
 Melee {type=boolean}
]
