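# Animation Blueprint for the face accessory part of the male CosmicTrellis Goth cosmetic body, adding secondary motion to its hair bones.
# It copies the pose from the attached parent mesh and passes it through the part's Control Rig, fed with crouching and skydiving states.
# Nine chained Anim Dynamics nodes then spring hair_01 to hair_09 under custom gravity.
# The first four simulate root-relative with stronger vertical response, while the rest use weaker, uniform motion response.
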AnimBlueprint /BRCosmetics/Characters/Player/Male/Medium/Bodies/M_MED_CosmicTrellis_Goth/Meshes/Parts/M_MED_CosmicTrellis_Goth_FaceAcc_AnimBP
target_skeleton=&/M_MED_CosmicTrellis_Goth_FaceAcc_Skeleton
graphs [
 AnimGraph {nodes [
   n1 = Control Rig!(Source=Copy Pose From Mesh() {bUseAttachedParent=true}, Alpha, crounching:bool=Get bIsCrouching(), Skidiving:bool=Get bIsSkyDiving()) -> (Pose) {controlRigAssetReference {controlRigAsset=&/M_MED_CosmicTrellis_Goth_FaceAcc_CtrlRig}}
   n2 = Anim Dynamics - Bone: hair_01(ComponentPose=Local To Component(LocalPose=@n1), Alpha=1, ExternalForce="0,0,0") {
    angularDampingOverride=1
    angularSpringConstant=2000
    bAngularSpring=true
    bOverrideAngularDamping=true
    bUseGravityOverride=true
    gravityScale=0.35
    simulationSpace=RootRelative
    alphaBoolBlend {blendOption=Linear}
    boundBone {boneName=hair_01}
    componentAppliedLinearAccClamp=<10000 10000 10000>
    componentLinearAccScale=<0.2 0.2 0.8>
    componentLinearVelScale=<0.2 0.2 0.8>
    gravityOverride {z=-500}
    physicsBodyDefinitions [{
      boundBone {boneName=hair_01}
      boxExtents=<12 12 12>
      constraintSetup {angularLimitsMax=<0 10 10>, angularLimitsMin=<0 0 -10>, angularTarget=<1 0 0>}
      localJointOffset=<1 0 0>
     }]
   }
   n3 = Anim Dynamics - Bone: hair_02(ComponentPose=@n2, Alpha=1, ExternalForce="0,0,0") {
    angularDampingOverride=0.9999
    angularSpringConstant=2000
    bAngularSpring=true
    bOverrideAngularDamping=true
    bUseGravityOverride=true
    gravityScale=0.35
    simulationSpace=RootRelative
    alphaBoolBlend {blendOption=Linear}
    boundBone {boneName=hair_02}
    componentAppliedLinearAccClamp=<10000 10000 10000>
    componentLinearAccScale=<0.2 0.2 0.8>
    componentLinearVelScale=<0.2 0.2 0.8>
    gravityOverride {z=-200}
    physicsBodyDefinitions [{
      boundBone {boneName=hair_02}
      boxExtents=<12 12 12>
      constraintSetup {angularLimitsMax=<0 10 20>, angularLimitsMin=<0 0 -10>, angularTarget=<1 0 0>}
      localJointOffset=<1 0 0>
     }]
   }
   n4 = Anim Dynamics - Bone: hair_03(ComponentPose=@n3, Alpha=1, ExternalForce="0,0,0") {
    angularDampingOverride=0.9999
    angularSpringConstant=2000
    bAngularSpring=true
    bOverrideAngularDamping=true
    bUseGravityOverride=true
    gravityScale=0.35
    simulationSpace=RootRelative
    alphaBoolBlend {blendOption=Linear}
    boundBone {boneName=hair_03}
    componentAppliedLinearAccClamp=<10000 10000 10000>
    componentLinearAccScale=<0.2 0.2 0.8>
    componentLinearVelScale=<0.2 0.2 0.8>
    gravityOverride {z=-200}
    physicsBodyDefinitions [{
      boundBone {boneName=hair_03}
      boxExtents=<12 12 12>
      constraintSetup {angularLimitsMax=<0 20 25>, angularLimitsMin=<0 -20 -25>, angularTarget=<1 0 0>}
      localJointOffset=<1 0 0>
     }]
   }
   n5 = Anim Dynamics - Bone: hair_04(ComponentPose=@n4, Alpha=1, ExternalForce="0,0,0") {
    angularDampingOverride=1
    angularSpringConstant=2000
    bAngularSpring=true
    bOverrideAngularDamping=true
    bUseGravityOverride=true
    gravityScale=0.35
    simulationSpace=RootRelative
    alphaBoolBlend {blendOption=Linear}
    boundBone {boneName=hair_04}
    componentAppliedLinearAccClamp=<10000 10000 10000>
    componentLinearAccScale=<0.2 0.2 0.8>
    componentLinearVelScale=<0.2 0.2 0.8>
    gravityOverride {z=-500}
    physicsBodyDefinitions [{
      boundBone {boneName=hair_04}
      boxExtents=<12 12 12>
      constraintSetup {angularLimitsMax=<0 10 20>, angularLimitsMin=<0 0 -20>, angularTarget=<1 0 0>}
      localJointOffset=<1 0 0>
     }]
   }
   n6 = Anim Dynamics - Bone: hair_05(ComponentPose=@n5, Alpha=1, ExternalForce="0,0,0") {
    angularDampingOverride=1
    angularSpringConstant=2000
    bAngularSpring=true
    bOverrideAngularDamping=true
    bUseGravityOverride=true
    gravityScale=0.35
    alphaBoolBlend {blendOption=Linear}
    boundBone {boneName=hair_05}
    componentAppliedLinearAccClamp=<3000 3000 3000>
    componentLinearAccScale=<0.1 0.1 0.1>
    componentLinearVelScale=<0.2 0.2 0.2>
    gravityOverride {z=-500}
    physicsBodyDefinitions [{
      boundBone {boneName=hair_05}
      boxExtents=<12 12 12>
      constraintSetup {angularLimitsMax=<0 10 10>, angularLimitsMin=<0 0 -10>, angularTarget=<1 0 0>}
      localJointOffset=<0.1 0 0>
     }]
   }
   n7 = Anim Dynamics - Bone: hair_06(ComponentPose=@n6, Alpha=1, ExternalForce="0,0,0") {
    angularDampingOverride=1
    angularSpringConstant=2000
    bAngularSpring=true
    bOverrideAngularDamping=true
    bUseGravityOverride=true
    gravityScale=0.35
    alphaBoolBlend {blendOption=Linear}
    boundBone {boneName=hair_06}
    componentAppliedLinearAccClamp=<3000 3000 3000>
    componentLinearAccScale=<0.1 0.1 0.1>
    componentLinearVelScale=<0.2 0.2 0.2>
    gravityOverride {z=-500}
    physicsBodyDefinitions [{
      boundBone {boneName=hair_06}
      boxExtents=<12 12 12>
      constraintSetup {angularLimitsMax=<0 10 10>, angularLimitsMin=<0 0 -10>, angularTarget=<1 0 0>}
      localJointOffset=<0.1 0 0>
     }]
   }
   n8 = Anim Dynamics - Bone: hair_07(ComponentPose=@n7, Alpha=1, ExternalForce="0,0,0") {
    angularDampingOverride=1
    angularSpringConstant=2000
    bAngularSpring=true
    bOverrideAngularDamping=true
    bUseGravityOverride=true
    gravityScale=0.35
    alphaBoolBlend {blendOption=Linear}
    boundBone {boneName=hair_07}
    componentAppliedLinearAccClamp=<3000 3000 3000>
    componentLinearAccScale=<0.1 0.1 0.1>
    componentLinearVelScale=<0.2 0.2 0.2>
    gravityOverride {z=-500}
    physicsBodyDefinitions [{
      boundBone {boneName=hair_07}
      boxExtents=<12 12 12>
      constraintSetup {angularLimitsMax=<0 10 10>, angularLimitsMin=<0 0 -10>, angularTarget=<1 0 0>}
      localJointOffset=<0.1 0 0>
     }]
   }
   n9 = Anim Dynamics - Bone: hair_08(ComponentPose=@n8, Alpha=1, ExternalForce="0,0,0") {
    angularDampingOverride=1
    angularSpringConstant=2000
    bAngularSpring=true
    bOverrideAngularDamping=true
    bUseGravityOverride=true
    gravityScale=0.35
    alphaBoolBlend {blendOption=Linear}
    boundBone {boneName=hair_08}
    componentAppliedLinearAccClamp=<3000 3000 3000>
    componentLinearAccScale=<0.1 0.1 0.1>
    componentLinearVelScale=<0.2 0.2 0.2>
    gravityOverride {z=-500}
    physicsBodyDefinitions [{
      boundBone {boneName=hair_08}
      boxExtents=<12 12 12>
      constraintSetup {angularLimitsMax=<0 10 10>, angularLimitsMin=<0 0 -10>, angularTarget=<1 0 0>}
      localJointOffset=<0.1 0 0>
     }]
   }
   n10 = Anim Dynamics - Bone: hair_09(ComponentPose=@n9, Alpha=1, ExternalForce="0,0,0") {
    angularDampingOverride=1
    angularSpringConstant=2000
    bAngularSpring=true
    bOverrideAngularDamping=true
    bUseGravityOverride=true
    gravityScale=0.35
    alphaBoolBlend {blendOption=Linear}
    boundBone {boneName=hair_09}
    componentAppliedLinearAccClamp=<3000 3000 3000>
    componentLinearAccScale=<0.1 0.1 0.1>
    componentLinearVelScale=<0.2 0.2 0.2>
    gravityOverride {z=-500}
    physicsBodyDefinitions [{
      boundBone {boneName=hair_09}
      boxExtents=<12 12 12>
      constraintSetup {angularLimitsMax=<0 10 10>, angularLimitsMin=<0 0 -10>, angularTarget=<1 0 0>}
      localJointOffset=<0.1 0 0>
     }]
   }
   Output Pose(Result=Component To Local(ComponentPose=@n10)) {layerGroup=None}
  ]}
 EventGraph {nodes [Event BlueprintUpdateAnimation()]}
]
variables []
